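# Sound cue for looping outer water ambience at the Cube Island event, on the events sound class with a three-voice concurrency cap.
# It mixes a non-spatialized close layer with a quieter, spatialized, wide-stereo distant layer.
# Custom attenuation curves crossfade them: close fades out by half the falloff distance, where distant peaks before fading to zero.
# Quality-level nodes swap in optimized waves, using a generic waterfall loop for the close layer.
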SoundCue /Figment_S08/Athena/Sounds/Events/CubeIsland/CubeIsland_Water_Outer_Cue
root=Mixer_0
nodes [
 {
  id=Attenuation_0
  children [QualityLevel_0]
  properties {
   bOverrideAttenuation=true
   attenuationOverrides {
    bSpatialize=false
    distanceAlgorithm=Custom
    falloffDistance=5500
    attenuationShapeExtents {x=0}
    customAttenuationCurve {editorCurveData {keys [<at 0 1>,<at 0.5 0>]}}
   }
  }
 }
 {
  id=Attenuation_1
  children [QualityLevel_1]
  properties {
   bOverrideAttenuation=true
   attenuationOverrides {
    distanceAlgorithm=Custom
    falloffDistance=5500
    stereoSpread=7500
    attenuationShapeExtents {x=0}
    customAttenuationCurve {editorCurveData {keys [<at 0 0>,<at 0.5 1>,<at 1 0>]}}
   }
  }
 }
 {id=Mixer_0, children [Attenuation_0, Attenuation_1], properties {inputVolume [0.8,0.4]}}
 {id=Mixer_1, children [WavePlayer_3], properties {inputVolume [0.7]}}
 {id=QualityLevel_0, children [WavePlayer_2,WavePlayer_2,Mixer_1,Mixer_1]}
 {id=QualityLevel_1, children [WavePlayer_0,WavePlayer_0,WavePlayer_1,WavePlayer_1]}
 {id=WavePlayer_0, properties {bLooping=true, soundWaveAssetPtr=/Game/Athena/Sounds/Events/CubeIsland/CubeIsland_Water_Distant_01}}
 {id=WavePlayer_1, properties {bLooping=true, soundWaveAssetPtr=/Game/Athena/Sounds/Events/CubeIsland/CubeIsland_Water_Distant_01_opt}}
 {id=WavePlayer_2, properties {bLooping=true, soundWaveAssetPtr=/Game/Athena/Sounds/Events/CubeIsland/CubeIsland_Water_Close_01}}
 {id=WavePlayer_3, properties {bLooping=true, soundWaveAssetPtr=/Game/Sounds/Fort_Ambient_Looping/Water/waterfall_new_ST_01_opt}}
]
properties {
 bOverrideConcurrency=true
 maxDistance=5500
 soundClassObject=&Fort_Events
 volumeMultiplier=0.45
 attenuationOverrides {bSpatialize=false, falloffDistance=5000, nonSpatializedRadiusStart=500, stereoSpread=5000, attenuationShapeExtents {x=0}}
 concurrencyOverrides {maxCount=3, resolutionRule=StopQuietest}
}
